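# Physics asset for the skeletal mesh of the Barbarian pickaxe, a two-handed melee weapon cosmetic.
# It approximates the weapon with two capsule bodies: a long one on the Melee_TwoHanded_Handle bone and a small one on the dyn_ring_00 ring bone.
# No constraints are defined, so nothing in this asset joints the ring body to the handle.
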PhysicsAsset /BRCosmetics/Weapons/FORT_Melee/Pickaxe_Barbarian/Meshes/SK_Pickaxe_Barbarian_Physics
bodies {
 Melee_TwoHanded_Handle [{shapeType=Capsule, length=104.5524, radius=4.45, center=<0.74 -1.6131 30.1703>, rotation=<rot 0.1967 -90 0.0395>}]
 dyn_ring_00 [{shapeType=Capsule, length=4.5247, radius=3.6087, center=<2.2118 -0.0103 0.8415>, rotation=<rot 0 90.5943 -81.4431>}]
}
constraints []
